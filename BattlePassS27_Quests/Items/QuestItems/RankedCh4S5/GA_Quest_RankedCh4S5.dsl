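; Gameplay ability for Ranked Chapter 4 Season 5 that grants the player one random Urgent transient quest per match in Habanero playlists.
; It listens to the game phase stateful event until Aircraft or SafeZones, resolving the controller or awaiting possession.
; It then waits for the bus jump, briefly delaying late joiners already skydiving.
; CheckEnabledQuests keeps candidates with positive curve table values, and a random one is granted.
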(event Ability|EventActivateAbility
  ;[0 0 0] Wait until the Aircraft phase has started, or the Zone phase if the player joined the match late
  (|CallParentFunction)
  (bind (_gamestate _playlist _playlistcontexttags) (Class|FortAsyncActionCurrentPlaylistReady|PlaylistDataReadyAsync)
    (:OnReady
      (if (GameplayTags|HasTag _playlistcontexttags "(TagName=\"Athena.Playlist.Habanero\")" true)
        (Variables|Runtime|SetAthenaGameState (Fortnite|Athena|GameState|GetGameStateAthenafromContext (Variables|GAVars|Data|GetOwningPlayerPawn)))
        (bind _returnvalue (GameplayEventRouter|GetEventRouter "/Script/FortniteGame.FortGameState" (Variables|GAVars|Data|GetOwningPlayerPawn)))
        (bind (_asynctaskproxy _payload) (GameplayEventRouter|StartListeningToStatefulEvent _returnvalue "/Script/FortniteGame.FortGamePhaseUpdatedEvent")
          (:then
            ;;[1 1 1] NOTE: This may get set after having run OnGamePhaseChanged (if SavedEventStateExists runs)
            (Variables|Runtime|SetGamePhaseUpdatedAsyncTask _asynctaskproxy))
          (:OnEventReceived
            _)
          (:OnSavedEventStateExists
            _))
        (CallFunction|OnGamePhaseChanged (Utilities|Struct|BreakFortGamePhaseUpdatedEvent _payload))))))

(event Custom|OnGamePhaseChanged
  (params (GamePhase :type EAthenaGamePhase%20Enum))
  ;[0 0 0] Wait until the Aircraft phase has started, or the Zone phase if the player joined the match late
  (if (CallFunction|HasMatchStarted GamePhase)
    (bind _gamephaseupdatedasynctask (Variables|Runtime|GetGamePhaseUpdatedAsyncTask)
      (:then
        (GameplayEventRouter|StopListeningtoStatefulEvent _gamephaseupdatedasynctask))
      (:else
        _))
    (CallFunction|StartGrantUrgentQuestProcess)))

(event Custom|StartGrantUrgentQuestProcess
  ;[0 0 0] Grant the Urgent quest when the player jumps from the Bus, or immediately if they load after having jumped
  ;;[1 1 1] Protect against StopListening running too early and not cleaning the listener properly the first time
  (Utilities|FlowControl|DoOnce false
    (:Completed
      ;;[1 1 1] Ensure we have our FortPlayerControllerAthena set before moving forward
      (bind _asfort_player_controller_athena (Utilities|Casting|CastToFortPlayerControllerAthena (Pawn|GetController (Variables|GAVars|Data|GetOwningPlayerPawn)))
        (:then
          (Variables|Runtime|SetOwningPlayerController _asfort_player_controller_athena)
          (if (Aircraft|IsinAircraft (Variables|Runtime|GetOwningPlayerController) :!class /Script/FortniteGame.FortPlayerController)
            (Ability|Tasks|WaitGameplayTagAdd "(TagName=\"MovementMode.Skydiving\")" 0 true
              (:Added
                _))
            (else
              (if (Parachute|IsSkydivingfromBus (Variables|GAVars|Data|GetOwningPlayerPawn))
                ;;[1 1 1] The player loaded late into the match, so wait a little bit for them to catch up on the UI (Delay isn't needed for gameplay and could be removed)
                (Utilities|FlowControl|Delay 2.0)
                (else
                  _))))
          (CallFunction|CheckEnabledQuests)
          (bind _output_get (Variables|Runtime|SetUrgentQs2QuestToGrant (Utilities|Array|RandomArrayItem (Variables|Runtime|GetUrgentQs2QuestToGrant_Array_Enabled))))
          (Class|FortControllerComponentTransientQuests|GrantTransientQuest (Class|FortPlayerControllerAthena|GetTransientQuestsComponent (Variables|Runtime|GetOwningPlayerController)) _output_get 0))
        (:CastFailed
          (Pawn|BindEventtoReceiveControllerChangedDelegate (Variables|GAVars|Data|GetOwningPlayerPawn) (delegate Custom|OnControllerChanged)))))))

(event Custom|OnControllerChanged
  (params (Pawn :type Pawn%20Object%20Reference) (OldController :type Controller%20Object%20Reference) (NewController :type Controller%20Object%20Reference))
  ;[0 0 0] Grant the Urgent quest when the player jumps from the Bus, or immediately if they load after having jumped
  ;;[1 1 1] Ensure we have our FortPlayerControllerAthena set before moving forward
  (bind _asfort_player_controller_athena (Utilities|Casting|CastToFortPlayerControllerAthena NewController))
  (Variables|Runtime|SetOwningPlayerController _asfort_player_controller_athena)
  (Pawn|UnbindEventfromReceiveControllerChangedDelegate (Variables|GAVars|Data|GetOwningPlayerPawn) (delegate Custom|OnControllerChanged))
  (if (Aircraft|IsinAircraft (Variables|Runtime|GetOwningPlayerController) :!class /Script/FortniteGame.FortPlayerController)
    (Ability|Tasks|WaitGameplayTagAdd "(TagName=\"MovementMode.Skydiving\")" 0 true
      (:Added
        _))
    (else
      (if (Parachute|IsSkydivingfromBus (Variables|GAVars|Data|GetOwningPlayerPawn))
        ;;[1 1 1] The player loaded late into the match, so wait a little bit for them to catch up on the UI (Delay isn't needed for gameplay and could be removed)
        (Utilities|FlowControl|Delay 2.0)
        (else
          _))))
  (CallFunction|CheckEnabledQuests)
  (bind _output_get (Variables|Runtime|SetUrgentQs2QuestToGrant (Utilities|Array|RandomArrayItem (Variables|Runtime|GetUrgentQs2QuestToGrant_Array_Enabled))))
  (Class|FortControllerComponentTransientQuests|GrantTransientQuest (Class|FortPlayerControllerAthena|GetTransientQuestsComponent (Variables|Runtime|GetOwningPlayerController)) _output_get 0))

(event Ability|EventOnEndAbility
  (params (bWasCancelled :type Boolean))
  (|CallParentFunction bWasCancelled)
  (bind _gamephaseupdatedasynctask (Variables|Runtime|GetGamePhaseUpdatedAsyncTask)
    (:then
      (GameplayEventRouter|StopListeningtoStatefulEvent _gamephaseupdatedasynctask))
    (:else)))

(fn CheckEnabledQuests
  (for _array_element (Variables|Setup|GetUrgentQs2QuestToGrant_Array)
    (Variables|Runtime|SetPotentialQuest _array_element)
    (bind _outxy (DataTable|EvaluateCurveTableRow "/BattlePassS27_Quests/Items/QuestItems/RankedCh4S5/RankedCh4S5.RankedCh4S5" (Utilities|String|StringToName (Item|GetPersistentName _array_element)))
      (:RowFound
        (if (> _outxy 0)
          (Utilities|Array|AddUnique (Variables|Runtime|GetUrgentQs2QuestToGrant_Array_Enabled) (Variables|Runtime|GetPotentialQuest))))))
  (return))

(fn HasMatchStarted
  (params (GamePhase :type EAthenaGamePhase%20Enum))
  (return (or (Utilities|Enum|Equal(Enum) GamePhase "Aircraft") (Utilities|Enum|Equal(Enum) GamePhase "SafeZones"))))
